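# Animation Blueprint for the MincePounce cosmetic tail accessory on the shared Fortnite base tail skeleton, adjusting the tail's pose and visibility.
# It layers constant additive root and base rotations onto the input pose, then blends replacement rotations along the tail chain for sliding.
# The sliding blend alpha is interpolated each update from bIsSliding, rising faster than it falls.
# When character customization loads, it hides the MiscOrTail mesh if the backbling is a cape.
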AnimBlueprint /BRCosmetics/Accessories/FORT_Tails/Tail_F_MED_MincePounce/Meshes/MincePounce_Tail_AnimBP
target_skeleton=/Game/Accessories/FORT_Tails/Common/Fortnite_Base_Tail_Skeleton
graphs [
 AnimGraph {nodes [
   n1 = "Transform (Modify) Bone - Bone: C_Root_Main_Root_Jnt"(ComponentPose=Local To Component(LocalPose=Input Pose()), Alpha=1, Translation="0,0,0", Rotation="0,0,-30", Scale="1,1,1") {rotationMode=BMM_Additive, rotationSpace=BCS_BoneSpace, boneToModify {boneName=C_Root_Main_Root_Jnt}, rotation {roll=-30}}
   n2 = "Transform (Modify) Bone - Bone: C_Tail_A_Base_Jnt"(ComponentPose=@n1, Alpha=1, Translation="0,0,0", Rotation="0,30,0", Scale="1,1,1") {rotationMode=BMM_Additive, rotationSpace=BCS_BoneSpace, boneToModify {boneName=C_Tail_A_Base_Jnt}, rotation {yaw=30}}
   n3 = "Transform (Modify) Bone - Bone: C_Tail_A_Base_Jnt"(ComponentPose=@n2, Alpha=Get tail rot sliding() -> (tail rot sliding:double), Translation="0,0,0", Rotation="0,55,0", Scale="1,1,1") {rotationMode=BMM_Replace, rotationSpace=BCS_BoneSpace, boneToModify {boneName=C_Tail_A_Base_Jnt}, rotation {yaw=55}}
   n4 = Get tail rot sliding()
   n5 = "Transform (Modify) Bone - Bone: C_Tail_A_4_Jnt"(
    ComponentPose="Transform (Modify) Bone - Bone: C_Tail_A_2_Jnt"(
     ComponentPose="Transform (Modify) Bone - Bone: C_Tail_A_1_Jnt"(ComponentPose=@n3, Alpha=@n4, Translation="0,0,0", Rotation="20,0,0", Scale="1,1,1") {rotationMode=BMM_Replace, rotationSpace=BCS_BoneSpace, boneToModify {boneName=C_Tail_A_1_Jnt}, rotation {pitch=20}}
     Alpha=Get tail rot sliding()
     Translation="0,0,0"
     Rotation="40,0,0"
     Scale="1,1,1"
    ) {rotationMode=BMM_Replace, rotationSpace=BCS_BoneSpace, boneToModify {boneName=C_Tail_A_2_Jnt}, rotation {pitch=40}}
    Alpha=Get tail rot sliding()
    Translation="0,0,0"
    Rotation="20,0,0"
    Scale="1,1,1"
   ) {rotationMode=BMM_Replace, rotationSpace=BCS_BoneSpace, boneToModify {boneName=C_Tail_A_4_Jnt}, rotation {pitch=20}}
   n6 = Get tail rot sliding()
   Output Pose(Result=Component To Local(ComponentPose="Transform (Modify) Bone - Bone: C_Tail_A_6_Jnt"(ComponentPose=@n5, Alpha=@n6, Translation="0,0,0", Rotation="-20,0,0", Scale="1,1,1") {rotationMode=BMM_Replace, rotationSpace=BCS_BoneSpace, boneToModify {boneName=C_Tail_A_6_Jnt}, rotation {pitch=-20}}))
  ]}
 EventGraph {nodes [
   n1 = Event OnCharacterCustomizationLoaded()
   > n2 = Set tail_skeletal mesh_component(tail_skeletal mesh_component:Skeletal Mesh Component Budgeted Object Reference=Get Array Item with EFort Custom Part Type(Array=@n1.CustomMeshParts, PartType=MiscOrTail)) -> (Output_Get:Skeletal Mesh Component Budgeted Object Reference)
   > n3 = Set is_wearing_cape(is_wearing_cape:bool=CheckForBackbling().bIsBackblingACape) -> (Output_Get:bool)
   > n4 = Branch(Condition=@n3)
   > SetVisibility!(execute, self:Scene Component Object Reference=@n2, bNewVisibility:bool=false, bPropagateToChildren:bool=false) -> (then)
   SetVisibility(execute=@n4.else, self=@n2, bNewVisibility=true, bPropagateToChildren=false)
   n5 = Event BlueprintUpdateAnimation()
   > Set Delta Time X(Delta Time X=@n5.DeltaTimeX)
   > n6 = Branch(Condition=Get bIsSliding())
   > Set tail rot sliding(tail rot sliding:double=FInterpTo(Current=Get tail rot sliding(), Target=1, DeltaTime=Get Delta Time X(), InterpSpeed=8)) -> (Output_Get:double)
   Set tail rot sliding(execute=@n6.else, tail rot sliding=FInterpTo(Current=Get tail rot sliding(), Target=0, DeltaTime=Get Delta Time X(), InterpSpeed=5))
   Get material 1() -> (material 1:Material Instance Dynamic Object Reference)
  ]}
]
variables [
 Delta Time X {type=number}
 is_wearing_cape {type=boolean}
 material 1 {type=MaterialInstanceDynamic}
 material 2 {type=MaterialInstanceDynamic}
 tail rot sliding {type=number}
 tail_skeletal mesh_component {type=SkeletalMeshComponentBudgeted}
]
